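; UI director component that shows the Seven Slide reticle scene while the local player's viewed pawn has the slider in its inventory.
; Binds to the Athena controller's view-target changes, caching the new pawn and starting gameplay-tag add/remove wait tasks on it.
; Tag addition adds the scene and broadcasts the local player's slider item as a stateful router event.
; Tag removal or EndPlay removes the scene; EndPlay also unbinds the delegate.
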(event EventOnReadyToShowUI
  ; Ready to Show / End Play
  (bind _asathena_player_controller (Utilities|Casting|CastToAthena_PlayerController (CallFunction|GetOwningLocalPlayerController)))
  (bind _output_get (Variables|Default|SetAthenaPlayerController _asathena_player_controller))
  (EventDispatchers|BindEventtoOnViewTargetChanged _output_get (EventDispatchers|CreateEvent :!event Custom|OnViewTargetChanged)))

(event EventEndPlay
  (params (EndPlayReason :type EEndPlayReason%20Enum :default "Destroyed"))
  (bind _athenaplayercontroller (Variables|Default|GetAthenaPlayerController))
  ; Ready to Show / End Play
  (CallFunction|AddorRemoveScene (Variables|Scenes|GetSevenSlideReticleScene))
  (Utilities|IsValid _athenaplayercontroller
    (:Is%20Valid
      (EventDispatchers|UnbindEventfromOnViewTargetChanged _athenaplayercontroller (EventDispatchers|CreateEvent :!event Custom|OnViewTargetChanged)))))

(event Custom|OnViewTargetChanged
  (params (PlayerController :type Fort%20Player%20Controller%20Object%20Reference) (OldViewTarget :type Actor%20Object%20Reference) (NewViewTarget :type Actor%20Object%20Reference))
  (bind _sevenslidereticlescene (Variables|Scenes|GetSevenSlideReticleScene))
  (bind _asfort_player_pawn (Utilities|Casting|CastToFortPlayerPawn NewViewTarget))
  (bind _output_get (Variables|Default|SetPlayerPawn _asfort_player_pawn))
  (Utilities|FlowControl|Sequence
    (:then_0
      ; Slider added to Inventory
      (Ability|Tasks|WaitGameplayTagAddToActor _output_get (Variables|Default|GetSevenSlideInInventoryTag)
        (:Added
          (CallFunction|AddorRemoveScene _sevenslidereticlescene true)
          (bind (_contains _item) (CallFunction|GetPlayerSevenSlide))
          (if _contains
            (CallFunction|BroadcastSevenSlideItem _item)))))
    (:then_1
      ; Slider Removed from Inventory
      (Ability|Tasks|WaitGameplayTagRemoveFromActor _output_get (Variables|Default|GetSevenSlideInInventoryTag)
        (:Removed
          (CallFunction|AddorRemoveScene _sevenslidereticlescene))))))

(fn BroadcastSevenSlideItem
  (params (ItemReference :type Fort%20Item%20Object%20Reference))
  (GameplayEventRouter|BroadcastStatefulEvent (GameplayEventRouter|GetGlobalEventRouter) (Utilities|Struct|MakeFortSevenSliderSliderItemEvent (Utilities|ResolveSoftReference ItemReference)) 0))

(fn GetPlayerSevenSlide
  (bind _asfort_player_controller (Utilities|Casting|CastToFortPlayerController (CallFunction|GetOwningLocalPlayerController))
    (:then
      (bind _returnvalue (Inventory|GetInventoryItemwithTag _asfort_player_controller (GameplayTags|MakeLiteralGameplayTag "(TagName=\"sevenslide.WID\")")))
      (return (Utilities|IsValid _returnvalue :!class /Script/Engine.KismetSystemLibrary) _returnvalue))
    (:CastFailed
      (return false 0))))
